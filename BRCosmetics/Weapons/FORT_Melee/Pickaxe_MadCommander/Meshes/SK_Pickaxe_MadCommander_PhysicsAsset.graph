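# Physics asset for the Mad Commander pickaxe skeletal mesh, a melee weapon cosmetic.
# A single capsule body on the Melee_TwoHanded_Handle bone, offset from the bone and rotated, serves as the weapon's collision shape.
# It defines no constraints, so the asset has no jointed or articulated parts.
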PhysicsAsset /BRCosmetics/Weapons/FORT_Melee/Pickaxe_MadCommander/Meshes/SK_Pickaxe_MadCommander_PhysicsAsset
bodies {Melee_TwoHanded_Handle [{shapeType=Capsule, length=66.4283, radius=38.3094, center=<0.1634 -8.9865 21.1742>, rotation=<rot -3.5393 -90 0.0528>}]}
constraints []
